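# Master material for the CoyoteTrail backpack cosmetic, built on the shared character pre-FX, skin FX and post-FX function chain.
# A useCoyoteReactive static switch overrides base color and emissive on the material attributes.
# The override uses the body's CoyoteTrail reveal function, an FX mask and masked Fresnel glow to blend in alternate diffuse and emissive looks.
# Scalar parameters select between pink, black and green variant colors through lerps rather than switches.
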Material /BRCosmetics/Accessories/FORT_Backpacks/Backpack_M_MED_CoyoteTrail/Materials/M_MED_CoyoteTrail_Backpack_MASTER
graph {
 properties {
  bAutomaticallySetUsageInEditor=false
  bUseMaterialAttributes=true
  bUsedWithMeshDeformer=true
  bUsedWithMorphTargets=true
  bUsedWithSkeletalMesh=true
  bUsedWithStaticLighting=true
  pixelDepthOffsetMode=PDOM_Legacy
  refractionMethod=RM_IndexOfRefraction
  shadingModel=MSM_FromMaterialExpression
  subsurfaceProfile=&SS_HeroSkin_02
  usedShadingModels=Default Lit | Subsurface | Clear Coat | Subsurface Profile | Cloth | Eye
 }
 nodes [
  n1 = MaterialFunctionCall() {materialFunction=&BaseFN_CharacterFunction_PreFX}
  n2 = MaterialFunctionCall(
   PreskinnedPosition=@n1.Pre skinned Local Position
   Gmap=@n1.Pre skinned Local Position
   MaterialAttributes=@n1.Pre skinned Local Position
   AO=@n1.Pre skinned Local Position
  ) {materialFunction=&CharacterSkinFX}
  n3 = GetMaterialAttributes(None=@n2.Result) {attributeGetTypes [#1, #2]}
  n4 = NamedRerouteUsage()
  n5 = NamedRerouteUsage()
  n6 = MaterialFunctionCall(
   AO=@n1.AO
   Distance Blend=@n1.AO
   Fresnel=@n1.AO
   Gmap=@n1.AO
   Pre Skinned Local Position=@n1.AO
   Material=StaticSwitchParameter(
    True=SetMaterialAttributes!(MaterialAttributes=@n3.MaterialAttributes, Base Color=@n4, Emissive Color=@n5) {attributeSetTypes [#1, #2]}
    False=@n2.Result
   ) {group=CoyoteTrail, parameterName=useCoyoteReactive}
  ) {materialFunction=&BaseFN_CharacterFunction_PostFX}
  MaterialFunctionCall(In=@n6.Result) {materialFunction=&CINE_MF_CharPostFX}
  n7 = MaterialFunctionCall() {materialFunction=/BRCosmetics/Characters/Player/Male/Medium/Bodies/M_MED_CoyoteTrail_Alt/Materials/MF_CoyoteTrail_Reveal}
  NamedRerouteDeclaration(None=@n7.Reveal) {name=outReveal, nodeColor=<rgba 0 1 0.6016 1>}
  NamedRerouteDeclaration(None=@n7.EdgeMask) {name=outEdgeMask, nodeColor=<rgba 0.1484 0 1 1>}
  n8 = TextureSampleParameter2D() {
   group=CoyoteTrail
   parameterName=Mask_FX
   samplerType=SAMPLERTYPE_Masks
   texture=/BRCosmetics/Accessories/FORT_Backpacks/Backpack_M_MED_CoyoteTrail/Textures/T_M_MED_CoyoteTrail_Backpack_FX
  }
  n9 = NamedRerouteDeclaration(None=@n8.R) {name=maskFX, nodeColor=<rgba 0.1953 0 1 1>}
  n10 = Saturate(None=Fresnel(ExponentIn=ScalarParameter() {defaultValue=2, group=CoyoteTrail, parameterName=Fresnel_Size}) {baseReflectFraction=0})
  NamedRerouteDeclaration(None=Multiply(A=Multiply(A=@n9, B=@n10), B=ScalarParameter() {defaultValue=5, group=CoyoteTrail, parameterName=Fresnel_Power}) {constB=25}) {name=maskfresnel, nodeColor=<rgba 0.1719 1 0 1>}
  n11 = NamedRerouteUsage()
  n12 = LinearInterpolate(
   A=NamedRerouteUsage()
   B=Multiply(A=VectorParameter() {group=CoyoteTrail, parameterName=VariantPink_DiffuseColor, defaultValue=<rgba 0.3763 1 0.0497 1>}.RGB, B=@n11)
   Alpha=@n11
  )
  n13 = VectorParameter() {group=CoyoteTrail, parameterName=VariantBlack_DiffuseColor, defaultValue=<rgba 0.0018 0.4452 0.0908 1>}
  n14 = NamedRerouteUsage()
  n15 = TextureSampleParameter2D() {
   group=CoyoteTrail
   parameterName=DefaultLook_Diffuse
   texture=/BRCosmetics/Accessories/FORT_Backpacks/Backpack_M_MED_CoyoteTrail/Textures/T_M_MED_CoyoteTrail_Backpack_D
  }
  n16 = Multiply(A=@n15.RGB, B=ScalarParameter() {defaultValue=0.2, parameterName=Tint Power})
  n17 = OneMinus(None=NamedRerouteUsage())
  n18 = LinearInterpolate(
   A=LinearInterpolate(A=NamedRerouteUsage(), B=Multiply(A=@n13.RGB, B=@n14), Alpha=@n14)
   B=NamedRerouteDeclaration(None=LinearInterpolate(A=@n15.RGB, B=@n16, Alpha=@n17)) {name=altDiffuse, nodeColor=<rgba 1 0 0.7266 1>}
   Alpha=NamedRerouteUsage()
  )
  n19 = NamedRerouteUsage()
  NamedRerouteDeclaration(None=LinearInterpolate(A=NamedRerouteUsage(), B=LinearInterpolate(A=@n12, B=@n18, Alpha=@n19), Alpha=NamedRerouteUsage())) {name=outBaseColor, nodeColor=<rgba 1 0.8203 0 1>}
  n20 = NamedRerouteUsage()
  n21 = NamedRerouteUsage()
  n22 = Multiply(
   A=VectorParameter() {group=CoyoteTrail, parameterName=VariantPink_EmissiveColor, defaultValue=<rgba 0.3763 1 0.0497 1>}.RGB
   B=NamedRerouteUsage()
  )
  n23 = Multiply(
   A=VectorParameter() {group=CoyoteTrail, parameterName=VariantPink_FresnelColor, defaultValue=<rgba 0.0946 0.7969 0.2021 1>}.RGB
   B=NamedRerouteUsage()
  )
  n24 = VectorParameter() {group=CoyoteTrail, parameterName=VariantGreen_EmissiveColor, defaultValue=<rgba 0.0685 0.5776 0.0908 1>}
  n25 = Multiply(
   A=@n24.RGB
   B=NamedRerouteDeclaration(None=ScalarParameter() {defaultValue=0.25, group=CoyoteTrail, parameterName=emissivePower}) {name=emissivePower, nodeColor=<rgba 0.3125 0 1 1>}
  )
  n26 = Multiply(
   A=VectorParameter() {group=CoyoteTrail, parameterName=VariantGreen_FresnelColor, defaultValue=<rgba 0.0648 0.5711 0.1441 1>}.RGB
   B=NamedRerouteUsage()
  )
  n27 = LinearInterpolate(
   A=LinearInterpolate(B=Add(A=@n22, B=@n23), Alpha=NamedRerouteUsage())
   B=LinearInterpolate(B=Add(A=@n25, B=@n26), Alpha=NamedRerouteUsage())
   Alpha=NamedRerouteDeclaration(None=ScalarParameter() {defaultValue=1, parameterName=isPinkAltBP}) {name=isPinkAlt, nodeColor=<rgba 1 0.3609 0.5636 1>}
  )
  n28 = NamedRerouteDeclaration(None=ScalarParameter() {defaultValue=1, parameterName=isBlackAltBP}) {name=isBlackAlt, nodeColor=<rgba 1 0.3047 0 1>}
  NamedRerouteDeclaration(None=LinearInterpolate(A=Add(A=@n20, B=@n21), B=LinearInterpolate(A=@n20, B=@n27, Alpha=@n28) {constAlpha=1}, Alpha=NamedRerouteUsage())) {name=outEmissive, nodeColor=<rgba 0 1 0.4609 1>}
  NamedRerouteDeclaration(None=@n3.Base Color) {name=inBaseColor, nodeColor=<rgba 0.5703 0 1 1>}
  NamedRerouteDeclaration(None=@n3.Emissive Color) {name=inEmissive, nodeColor=<rgba 0 0.7656 1 1>}
 ]
}
