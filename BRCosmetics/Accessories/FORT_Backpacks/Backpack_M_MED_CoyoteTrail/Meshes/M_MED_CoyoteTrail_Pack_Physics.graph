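# Physics asset for the CoyoteTrail backpack cosmetic, giving its left and right wing bones simulation bodies for secondary motion.
# Each side is a mirrored chain of three capsule bodies, two on the wing base and a thinner one on the wing bone.
# Tight swing and twist limits link each chain, with more swing allowed at the outer joint.
# No constraint connects the two sides or ties either chain root to another body.
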PhysicsAsset /BRCosmetics/Accessories/FORT_Backpacks/Backpack_M_MED_CoyoteTrail/Meshes/M_MED_CoyoteTrail_Pack_Physics
bodies {
 dyn_wing_base_l_01 [{shapeType=Capsule, length=3, radius=3, shapeName=dyn_wing_base_l_01_capsule, center=<5 0 0>, rotation=<rot 90 0 0>}]
 dyn_wing_base_l_02 [{shapeType=Capsule, length=3, radius=3, shapeName=dyn_wing_base_l_02_capsule, center=<5 0 0>, rotation=<rot 90 0 0>}]
 dyn_wing_base_r_01 [{shapeType=Capsule, length=3, radius=3, shapeName=dyn_wing_base_r_01_capsule, center=<5 0 0>, rotation=<rot 90 0 0>}]
 dyn_wing_base_r_02 [{shapeType=Capsule, length=3, radius=3, shapeName=dyn_wing_base_r_02_capsule, center=<5 0 0>, rotation=<rot 90 0 0>}]
 dyn_wing_l_01 [{shapeType=Capsule, length=2, radius=2, shapeName=dyn_wing_l_01_capsule, center=<5 0 0>, rotation=<rot 90 0 0>}]
 dyn_wing_r_01 [{shapeType=Capsule, length=2, radius=2, shapeName=dyn_wing_r_01_capsule, center=<5 0 0>, rotation=<rot 90 0 0>}]
}
constraints [
 {bone1Name=dyn_wing_base_l_02, bone2Name=dyn_wing_base_l_01, swing1LimitDegrees=2, swing2LimitDegrees=2, twistLimitDegrees=8}
 {bone1Name=dyn_wing_l_01, bone2Name=dyn_wing_base_l_02, swing1LimitDegrees=10, swing2LimitDegrees=5, twistLimitDegrees=8}
 {bone1Name=dyn_wing_base_r_02, bone2Name=dyn_wing_base_r_01, swing1LimitDegrees=2, swing2LimitDegrees=2, twistLimitDegrees=8}
 {bone1Name=dyn_wing_r_01, bone2Name=dyn_wing_base_r_02, swing1LimitDegrees=10, swing2LimitDegrees=5, twistLimitDegrees=8}
]
