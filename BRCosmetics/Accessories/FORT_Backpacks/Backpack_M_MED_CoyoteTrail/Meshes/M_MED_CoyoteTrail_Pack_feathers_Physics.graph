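# Physics asset for the CoyoteTrail backpack's feathers, setting up capsule bodies for dynamic simulation of its left and right wing bones.
# Each side has a master body with zero-limit swing and twist to the wing root and a looser link to a lower base body.
# Five four-segment feather chains (a to e) branch from each root, with angular limits widening from very tight at the root to slightly looser toward the tips.
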PhysicsAsset /BRCosmetics/Accessories/FORT_Backpacks/Backpack_M_MED_CoyoteTrail/Meshes/M_MED_CoyoteTrail_Pack_feathers_Physics
bodies {
 dyn_wing_base_l_down_base [{shapeType=Capsule, length=1, radius=2, shapeName=dyn_wing_base_l_down_base_capsule, center=<3 0 0>, rotation=<rot 90 0 0>}]
 dyn_wing_base_l_master [{shapeType=Capsule, length=1, radius=1, shapeName=dyn_wing_base_l_master_capsule}]
 dyn_wing_base_r_down_base [{shapeType=Capsule, length=1, radius=2, shapeName=dyn_wing_base_r_down_base_capsule, center=<3 0 0>, rotation=<rot 90 0 0>}]
 dyn_wing_base_r_master [{shapeType=Capsule, length=1, radius=1, shapeName=dyn_wing_base_r_master_capsule}]
 dyn_wing_l_01 [{shapeType=Capsule, length=4, radius=6, shapeName=dyn_wing_l_01_capsule, center=<3 0 0>, rotation=<rot 90 0 0>}]
 dyn_wing_l_a_01 [{shapeType=Capsule, length=5, radius=2, shapeName=dyn_wing_l_a_01_capsule, center=<5 0 0>, rotation=<rot 90 0 0>}]
 dyn_wing_l_a_02 [{shapeType=Capsule, length=8, radius=2, shapeName=dyn_wing_l_a_02_capsule, center=<6 0 0>, rotation=<rot 90 0 0>}]
 dyn_wing_l_a_03 [{shapeType=Capsule, length=5, radius=2, shapeName=dyn_wing_l_a_03_capsule, center=<5 0 0>, rotation=<rot 90 0 0>}]
 dyn_wing_l_a_04 [{shapeType=Capsule, length=5, radius=2, shapeName=dyn_wing_l_a_04_capsule, center=<5 0 0>, rotation=<rot 90 0 0>}]
 dyn_wing_l_b_01 [{shapeType=Capsule, length=5, radius=2, shapeName=dyn_wing_l_b_01_capsule, center=<5 0 0>, rotation=<rot 90 0 0>}]
 dyn_wing_l_b_02 [{shapeType=Capsule, length=8, radius=2, shapeName=dyn_wing_l_b_02_capsule, center=<6 0 0>, rotation=<rot 90 0 0>}]
 dyn_wing_l_b_03 [{shapeType=Capsule, length=5, radius=2, shapeName=dyn_wing_l_b_03_capsule, center=<5 0 0>, rotation=<rot 90 0 0>}]
 dyn_wing_l_b_04 [{shapeType=Capsule, length=6.7918, radius=2, shapeName=dyn_wing_l_b_04_capsule, center=<6.3226 0 0>, rotation=<rot 90 0 0>}]
 dyn_wing_l_c_01 [{shapeType=Capsule, length=5, radius=2, shapeName=dyn_wing_l_c_01_capsule, center=<5 0 0>, rotation=<rot 90 0 0>}]
 dyn_wing_l_c_02 [{shapeType=Capsule, length=8, radius=2, shapeName=dyn_wing_l_c_02_capsule, center=<6 0 0>, rotation=<rot 90 0 0>}]
 dyn_wing_l_c_03 [{shapeType=Capsule, length=5, radius=2, shapeName=dyn_wing_l_c_03_capsule, center=<5 0 0>, rotation=<rot 90 0 0>}]
 dyn_wing_l_c_04 [{shapeType=Capsule, length=7.4792, radius=2, shapeName=dyn_wing_l_c_04_capsule, center=<6.6093 0 0>, rotation=<rot 90 0 0>}]
 dyn_wing_l_d_01 [{shapeType=Capsule, length=5, radius=2, shapeName=dyn_wing_l_d_01_capsule, center=<5 0 0>, rotation=<rot 90 0 0>}]
 dyn_wing_l_d_02 [{shapeType=Capsule, length=8, radius=2, shapeName=dyn_wing_l_d_02_capsule, center=<6 0 0>, rotation=<rot 90 0 0>}]
 dyn_wing_l_d_03 [{shapeType=Capsule, length=5, radius=2, shapeName=dyn_wing_l_d_03_capsule, center=<5 0 0>, rotation=<rot 90 0 0>}]
 dyn_wing_l_d_04 [{shapeType=Capsule, length=5, radius=2, shapeName=dyn_wing_l_d_04_capsule, center=<5 0 0>, rotation=<rot 90 0 0>}]
 dyn_wing_l_e_01 [{shapeType=Capsule, length=5, radius=2, shapeName=dyn_wing_l_e_01_capsule, center=<5 0 0>, rotation=<rot 90 0 0>}]
 dyn_wing_l_e_02 [{shapeType=Capsule, length=8, radius=2, shapeName=dyn_wing_l_e_02_capsule, center=<6 0 0>, rotation=<rot 90 0 0>}]
 dyn_wing_l_e_03 [{shapeType=Capsule, length=5, radius=2, shapeName=dyn_wing_l_e_03_capsule, center=<5 0 0>, rotation=<rot 90 0 0>}]
 dyn_wing_l_e_04 [{shapeType=Capsule, length=7.0491, radius=2, shapeName=dyn_wing_l_e_04_capsule, center=<6.0184 0 0>, rotation=<rot 90 0 0>}]
 dyn_wing_r_01 [{shapeType=Capsule, length=4, radius=6, shapeName=dyn_wing_r_01_capsule, center=<3 0 0>, rotation=<rot 90 0 0>}]
 dyn_wing_r_a_01 [{shapeType=Capsule, length=5, radius=2, shapeName=dyn_wing_r_a_01_capsule, center=<5 0 0>, rotation=<rot 90 0 0>}]
 dyn_wing_r_a_02 [{shapeType=Capsule, length=8, radius=2, shapeName=dyn_wing_r_a_02_capsule, center=<7 0 0>, rotation=<rot 90 0 0>}]
 dyn_wing_r_a_03 [{shapeType=Capsule, length=5, radius=2, shapeName=dyn_wing_r_a_03_capsule, center=<5 0 0>, rotation=<rot 90 0 0>}]
 dyn_wing_r_a_04 [{shapeType=Capsule, length=10.8106, radius=2, shapeName=dyn_wing_r_a_04_capsule, center=<8.1674 0 0>, rotation=<rot 90 0 0>}]
 dyn_wing_r_b_01 [{shapeType=Capsule, length=5, radius=2, shapeName=dyn_wing_r_b_01_capsule, center=<5 0 0>, rotation=<rot 90 0 0>}]
 dyn_wing_r_b_02 [{shapeType=Capsule, length=8, radius=2, shapeName=dyn_wing_r_b_02_capsule, center=<7 0 0>, rotation=<rot 90 0 0>}]
 dyn_wing_r_b_03 [{shapeType=Capsule, length=5, radius=2, shapeName=dyn_wing_r_b_03_capsule, center=<5 0 0>, rotation=<rot 90 0 0>}]
 dyn_wing_r_b_04 [{shapeType=Capsule, length=7.6165, radius=2, shapeName=dyn_wing_r_b_04_capsule, center=<6.8984 0 0>, rotation=<rot 90 0 0>}]
 dyn_wing_r_c_01 [{shapeType=Capsule, length=5, radius=2, shapeName=dyn_wing_r_c_01_capsule, center=<5 0 0>, rotation=<rot 90 0 0>}]
 dyn_wing_r_c_02 [{shapeType=Capsule, length=8, radius=2, shapeName=dyn_wing_r_c_02_capsule, center=<7 0 0>, rotation=<rot 90 0 0>}]
 dyn_wing_r_c_03 [{shapeType=Capsule, length=5, radius=2, shapeName=dyn_wing_r_c_03_capsule, center=<5 0 0>, rotation=<rot 90 0 0>}]
 dyn_wing_r_c_04 [{shapeType=Capsule, length=7.7825, radius=2, shapeName=dyn_wing_r_c_04_capsule, center=<6.219 0 0>, rotation=<rot 90 0 0>}]
 dyn_wing_r_d_01 [{shapeType=Capsule, length=5, radius=2, shapeName=dyn_wing_r_d_01_capsule, center=<5 0 0>, rotation=<rot 90 0 0>}]
 dyn_wing_r_d_02 [{shapeType=Capsule, length=8, radius=2, shapeName=dyn_wing_r_d_02_capsule, center=<7 0 0>, rotation=<rot 90 0 0>}]
 dyn_wing_r_d_03 [{shapeType=Capsule, length=5, radius=2, shapeName=dyn_wing_r_d_03_capsule, center=<5 0 0>, rotation=<rot 90 0 0>}]
 dyn_wing_r_d_04 [{shapeType=Capsule, length=4.0894, radius=2, shapeName=dyn_wing_r_d_04_capsule, center=<4.2962 0 0>, rotation=<rot 90 0 0>}]
 dyn_wing_r_e_01 [{shapeType=Capsule, length=5, radius=2, shapeName=dyn_wing_r_e_01_capsule, center=<5 0 0>, rotation=<rot 90 0 0>}]
 dyn_wing_r_e_02 [{shapeType=Capsule, length=8, radius=2, shapeName=dyn_wing_r_e_02_capsule, center=<7 0 0>, rotation=<rot 90 0 0>}]
 dyn_wing_r_e_03 [{shapeType=Capsule, length=5, radius=2, shapeName=dyn_wing_r_e_03_capsule, center=<5 0 0>, rotation=<rot 90 0 0>}]
 dyn_wing_r_e_04 [{shapeType=Capsule, length=7.349, radius=2, shapeName=dyn_wing_r_e_04_capsule, center=<6.0739 0 0>, rotation=<rot 90 0 0>}]
}
constraints [
 {bone1Name=dyn_wing_l_a_01, bone2Name=dyn_wing_l_01, swing1LimitDegrees=1, swing2LimitDegrees=1, twistLimitDegrees=1}
 {bone1Name=dyn_wing_l_a_02, bone2Name=dyn_wing_l_a_01, swing1LimitDegrees=2, swing2LimitDegrees=2, twistLimitDegrees=2}
 {bone1Name=dyn_wing_l_a_03, bone2Name=dyn_wing_l_a_02, swing1LimitDegrees=5, swing2LimitDegrees=5, twistLimitDegrees=2}
 {bone1Name=dyn_wing_l_a_04, bone2Name=dyn_wing_l_a_03, swing1LimitDegrees=5, swing2LimitDegrees=5, twistLimitDegrees=2}
 {bone1Name=dyn_wing_l_b_01, bone2Name=dyn_wing_l_01, swing1LimitDegrees=1, swing2LimitDegrees=1, twistLimitDegrees=1}
 {bone1Name=dyn_wing_l_b_02, bone2Name=dyn_wing_l_b_01, swing1LimitDegrees=2, swing2LimitDegrees=2, twistLimitDegrees=2}
 {bone1Name=dyn_wing_l_b_03, bone2Name=dyn_wing_l_b_02, swing1LimitDegrees=5, swing2LimitDegrees=5, twistLimitDegrees=2}
 {bone1Name=dyn_wing_l_b_04, bone2Name=dyn_wing_l_b_03, swing1LimitDegrees=5, swing2LimitDegrees=5, twistLimitDegrees=2}
 {bone1Name=dyn_wing_l_c_01, bone2Name=dyn_wing_l_01, swing1LimitDegrees=1, swing2LimitDegrees=1, twistLimitDegrees=1}
 {bone1Name=dyn_wing_l_c_02, bone2Name=dyn_wing_l_c_01, swing1LimitDegrees=2, swing2LimitDegrees=2, twistLimitDegrees=2}
 {bone1Name=dyn_wing_l_c_03, bone2Name=dyn_wing_l_c_02, swing1LimitDegrees=5, swing2LimitDegrees=5, twistLimitDegrees=2}
 {bone1Name=dyn_wing_l_c_04, bone2Name=dyn_wing_l_c_03, swing1LimitDegrees=5, swing2LimitDegrees=5, twistLimitDegrees=2}
 {bone1Name=dyn_wing_l_d_01, bone2Name=dyn_wing_l_01, swing1LimitDegrees=1, swing2LimitDegrees=1, twistLimitDegrees=1}
 {bone1Name=dyn_wing_l_d_02, bone2Name=dyn_wing_l_d_01, swing1LimitDegrees=2, swing2LimitDegrees=2, twistLimitDegrees=2}
 {bone1Name=dyn_wing_l_d_03, bone2Name=dyn_wing_l_d_02, swing1LimitDegrees=5, swing2LimitDegrees=5, twistLimitDegrees=2}
 {bone1Name=dyn_wing_l_d_04, bone2Name=dyn_wing_l_d_03, swing1LimitDegrees=5, swing2LimitDegrees=5, twistLimitDegrees=2}
 {bone1Name=dyn_wing_l_e_01, bone2Name=dyn_wing_l_01, swing1LimitDegrees=1, swing2LimitDegrees=1, twistLimitDegrees=1}
 {bone1Name=dyn_wing_l_e_02, bone2Name=dyn_wing_l_e_01, swing1LimitDegrees=2, swing2LimitDegrees=2, twistLimitDegrees=2}
 {bone1Name=dyn_wing_l_e_03, bone2Name=dyn_wing_l_e_02, swing1LimitDegrees=5, swing2LimitDegrees=5, twistLimitDegrees=2}
 {bone1Name=dyn_wing_l_e_04, bone2Name=dyn_wing_l_e_03, swing1LimitDegrees=5, swing2LimitDegrees=5, twistLimitDegrees=2}
 {bone1Name=dyn_wing_r_a_03, bone2Name=dyn_wing_r_a_02, swing1LimitDegrees=5, swing2LimitDegrees=5, twistLimitDegrees=2}
 {bone1Name=dyn_wing_r_a_04, bone2Name=dyn_wing_r_a_03, swing1LimitDegrees=5, swing2LimitDegrees=5, twistLimitDegrees=2}
 {bone1Name=dyn_wing_r_b_01, bone2Name=dyn_wing_r_01, swing1LimitDegrees=1, swing2LimitDegrees=1, twistLimitDegrees=1}
 {bone1Name=dyn_wing_r_b_02, bone2Name=dyn_wing_r_b_01, swing1LimitDegrees=2, swing2LimitDegrees=2, twistLimitDegrees=2}
 {bone1Name=dyn_wing_r_b_03, bone2Name=dyn_wing_r_b_02, swing1LimitDegrees=5, swing2LimitDegrees=5, twistLimitDegrees=2}
 {bone1Name=dyn_wing_r_b_04, bone2Name=dyn_wing_r_b_03, swing1LimitDegrees=5, swing2LimitDegrees=5, twistLimitDegrees=2}
 {bone1Name=dyn_wing_r_c_01, bone2Name=dyn_wing_r_01, swing1LimitDegrees=1, swing2LimitDegrees=1, twistLimitDegrees=1}
 {bone1Name=dyn_wing_r_c_02, bone2Name=dyn_wing_r_c_01, swing1LimitDegrees=2, swing2LimitDegrees=2, twistLimitDegrees=2}
 {bone1Name=dyn_wing_r_c_03, bone2Name=dyn_wing_r_c_02, swing1LimitDegrees=5, swing2LimitDegrees=5, twistLimitDegrees=2}
 {bone1Name=dyn_wing_r_c_04, bone2Name=dyn_wing_r_c_03, swing1LimitDegrees=5, swing2LimitDegrees=5, twistLimitDegrees=2}
 {bone1Name=dyn_wing_r_d_01, bone2Name=dyn_wing_r_01, swing1LimitDegrees=1, swing2LimitDegrees=1, twistLimitDegrees=1}
 {bone1Name=dyn_wing_r_d_02, bone2Name=dyn_wing_r_d_01, swing1LimitDegrees=2, swing2LimitDegrees=2, twistLimitDegrees=2}
 {bone1Name=dyn_wing_r_d_03, bone2Name=dyn_wing_r_d_02, swing1LimitDegrees=5, swing2LimitDegrees=5, twistLimitDegrees=2}
 {bone1Name=dyn_wing_r_d_04, bone2Name=dyn_wing_r_d_03, swing1LimitDegrees=5, swing2LimitDegrees=5, twistLimitDegrees=2}
 {bone1Name=dyn_wing_r_e_04, bone2Name=dyn_wing_r_e_03, swing1LimitDegrees=5, swing2LimitDegrees=5, twistLimitDegrees=2}
 {bone1Name=dyn_wing_r_e_01, bone2Name=dyn_wing_r_01, swing1LimitDegrees=1, swing2LimitDegrees=1, twistLimitDegrees=1}
 {bone1Name=dyn_wing_r_e_02, bone2Name=dyn_wing_r_e_01, swing1LimitDegrees=2, swing2LimitDegrees=2, twistLimitDegrees=2}
 {bone1Name=dyn_wing_r_e_03, bone2Name=dyn_wing_r_e_02, swing1LimitDegrees=5, swing2LimitDegrees=5, twistLimitDegrees=2}
 {bone1Name=dyn_wing_r_a_02, bone2Name=dyn_wing_r_a_01, swing1LimitDegrees=2, swing2LimitDegrees=2, twistLimitDegrees=2}
 {bone1Name=dyn_wing_r_a_01, bone2Name=dyn_wing_r_01, swing1LimitDegrees=1, swing2LimitDegrees=1, twistLimitDegrees=1}
 {bone1Name=dyn_wing_r_01, bone2Name=dyn_wing_base_r_master, swing1LimitDegrees=0, swing2LimitDegrees=0, twistLimitDegrees=0}
 {bone1Name=dyn_wing_base_r_down_base, bone2Name=dyn_wing_base_r_master, swing1LimitDegrees=10, swing2LimitDegrees=10, twistLimitDegrees=2}
 {bone1Name=dyn_wing_l_01, bone2Name=dyn_wing_base_l_master, swing1LimitDegrees=0, swing2LimitDegrees=0, twistLimitDegrees=0}
 {bone1Name=dyn_wing_base_l_down_base, bone2Name=dyn_wing_base_l_master, swing1LimitDegrees=10, swing2LimitDegrees=10, twistLimitDegrees=2}
]
