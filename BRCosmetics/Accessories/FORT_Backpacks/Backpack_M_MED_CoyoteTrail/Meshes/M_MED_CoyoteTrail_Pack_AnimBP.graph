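# Animation Blueprint for the CoyoteTrail backpack, posing and simulating its feathered wings from the wearer's movement state.
# EventGraph turns sprint, targeting, crouch, skydive and jump flags into dive, fly and jump alphas, and updates a gravity override.
# AnimGraph offsets the pack, runs body RigidBody physics, layers interpolated additive bone transforms for wing poses and feather fanning, then runs feather RigidBody physics.
# Both simulations blend partially and share that gravity.
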AnimBlueprint /BRCosmetics/Accessories/FORT_Backpacks/Backpack_M_MED_CoyoteTrail/Meshes/M_MED_CoyoteTrail_Pack_AnimBP
target_skeleton=&/M_MED_CoyoteTrail_Pack_Skeleton
graphs [
 AnimGraph {nodes [
   n1 = "Get FInterp(Alpha, (5: 5))"() -> ("FInterp(Alpha, (5: 5))":float)
   n2 = Get FlyPose_alpha() -> (FlyPose_alpha:double)
   n3 = "Transform (Modify) Bone - Bone: pack"(ComponentPose=Local To Component(LocalPose=Input Pose()), Translation="0,2,0", Rotation="0,0,15", Scale="0.8,0.8,0.8") {
    rotationMode=BMM_Additive
    rotationSpace=BCS_BoneSpace
    scaleMode=BMM_Additive
    scaleSpace=BCS_BoneSpace
    translationMode=BMM_Additive
    translationSpace=BCS_BoneSpace
    alphaBoolBlend {blendOption=Linear}
    boneToModify {boneName=pack}
    rotation {roll=15}
    scale=<0.8 0.8 0.8>
    translation {y=2}
   }
   n4 = RigidBody!(ComponentPose=@n3, Alpha=0.4, OverrideWorldGravity:Vector=Get GravityOverride(), ExternalForce) -> (Pose) {
    bOverrideWorldGravity=true
    overridePhysicsAsset=&/M_MED_CoyoteTrail_Pack_Physics
    alphaBoolBlend {blendOption=Linear}
    baseBoneRef {boneName=root}
    componentAppliedLinearAccClamp=<3000 3000 3000>
    componentLinearAccScale=<0.3 0.3 0.3>
    componentLinearVelScale=<0.3 0.3 0.3>
    overrideWorldGravity {z=-980}
   }
   n5 = Get DivePose_alpha() -> (DivePose_alpha:double)
   n6 = "Transform (Modify) Bone - Bone: dyn_wing_base_r_02"(ComponentPose=@n4, Alpha=@n5, Translation="0,0,0", Rotation="-10,10,-45", Scale="1,1,1") {
    rotationMode=BMM_Additive
    rotationSpace=BCS_BoneSpace
    scaleSpace=BCS_BoneSpace
    alphaBoolBlend {blendOption=Linear}
    alphaScaleBiasClamp {bInterpResult=true, interpSpeedDecreasing=5, interpSpeedIncreasing=5}
    boneToModify {boneName=dyn_wing_base_r_02}
    rotation=<rot -10 10 -45>
   }
   n7 = "Transform (Modify) Bone - Bone: dyn_wing_base_l_02"(ComponentPose=@n6, Alpha=@n5, Translation="0,0,0", Rotation="-10,-10,45", Scale="1,1,1") {
    rotationMode=BMM_Additive
    rotationSpace=BCS_BoneSpace
    scaleSpace=BCS_BoneSpace
    alphaBoolBlend {blendOption=Linear}
    alphaScaleBiasClamp {bInterpResult=true, interpSpeedDecreasing=5, interpSpeedIncreasing=5}
    boneToModify {boneName=dyn_wing_base_l_02}
    rotation=<rot -10 -10 45>
   }
   n8 = "Transform (Modify) Bone - Bone: dyn_wing_base_r_02"(ComponentPose=@n7, Alpha=@n2, Translation="0,0,0", Rotation="0,0,5", Scale="1,1,1") {
    rotationMode=BMM_Additive
    rotationSpace=BCS_BoneSpace
    scaleSpace=BCS_BoneSpace
    alphaBoolBlend {blendOption=Linear}
    alphaScaleBiasClamp {bInterpResult=true}
    boneToModify {boneName=dyn_wing_base_r_02}
    rotation {roll=5}
   }
   n9 = "Transform (Modify) Bone - Bone: dyn_wing_r_01"(ComponentPose=@n8, Alpha=@n2, Translation="0,0,0", Rotation="0,5,0", Scale="1,1,1") {
    rotationMode=BMM_Additive
    rotationSpace=BCS_BoneSpace
    scaleSpace=BCS_BoneSpace
    alphaBoolBlend {blendOption=Linear}
    alphaScaleBiasClamp {bInterpResult=true}
    boneToModify {boneName=dyn_wing_r_01}
    rotation {yaw=5}
   }
   n10 = "Transform (Modify) Bone - Bone: dyn_wing_base_l_02"(ComponentPose=@n9, Alpha=@n2, Translation="0,0,0", Rotation="0,0,-5", Scale="1,1,1") {
    rotationMode=BMM_Additive
    rotationSpace=BCS_BoneSpace
    scaleSpace=BCS_BoneSpace
    alphaBoolBlend {blendOption=Linear}
    alphaScaleBiasClamp {bInterpResult=true}
    boneToModify {boneName=dyn_wing_base_l_02}
    rotation {roll=-5}
   }
   n11 = "Transform (Modify) Bone - Bone: dyn_wing_l_01"(ComponentPose=@n10, Alpha=@n2, Translation="0,0,0", Rotation="0,-5,0", Scale="1,1,1") {
    rotationMode=BMM_Additive
    rotationSpace=BCS_BoneSpace
    scaleSpace=BCS_BoneSpace
    alphaBoolBlend {blendOption=Linear}
    alphaScaleBiasClamp {bInterpResult=true}
    boneToModify {boneName=dyn_wing_l_01}
    rotation {yaw=-5}
   }
   n12 = "Transform (Modify) Bone - Bone: pack"(
    ComponentPose=@n11
    Alpha=MapRangeClamped(Value=Get JumpPose_alpha() -> (JumpPose_alpha:double), InRangeA=0, InRangeB=1, OutRangeB=3)
    Translation="0,0,0"
    Rotation="0.000,0.000,-5"
   ) {
    rotationMode=BMM_Additive
    rotationSpace=BCS_BoneSpace
    scaleSpace=BCS_BoneSpace
    alphaBoolBlend {blendOption=Linear}
    alphaScaleBiasClamp {bInterpResult=true, interpSpeedDecreasing=5, interpSpeedIncreasing=5}
    boneToModify {boneName=pack}
    rotation {roll=-5}
   }
   n13 = "Transform (Modify) Bone - Bone: dyn_wing_l_01"(ComponentPose=@n12, Alpha=@n1, Translation="0,0,0", Rotation="0.000,-15,0", Scale="1,1,1") {
    rotationMode=BMM_Additive
    rotationSpace=BCS_BoneSpace
    scaleSpace=BCS_BoneSpace
    alphaBoolBlend {blendOption=Linear}
    alphaScaleBiasClamp {bInterpResult=true, interpSpeedDecreasing=5, interpSpeedIncreasing=5}
    boneToModify {boneName=dyn_wing_l_01}
    rotation {yaw=-15}
   }
   n14 = "Transform (Modify) Bone - Bone: dyn_wing_l_e_base"(ComponentPose=@n13, Alpha=@n1, Translation="0,0,0", Rotation="0,14,0", Scale="1,1,1") {
    rotationMode=BMM_Additive
    rotationSpace=BCS_BoneSpace
    scaleSpace=BCS_BoneSpace
    alphaBoolBlend {blendOption=Linear}
    alphaScaleBiasClamp {bInterpResult=true, interpSpeedDecreasing=5, interpSpeedIncreasing=5}
    boneToModify {boneName=dyn_wing_l_e_base}
    rotation {yaw=14}
   }
   n15 = "Transform (Modify) Bone - Bone: dyn_wing_l_d_base"(ComponentPose=@n14, Alpha=@n1, Translation="0,0,0", Rotation="0,10,0", Scale="1,1,1") {
    rotationMode=BMM_Additive
    rotationSpace=BCS_BoneSpace
    scaleSpace=BCS_BoneSpace
    alphaBoolBlend {blendOption=Linear}
    alphaScaleBiasClamp {bInterpResult=true, interpSpeedDecreasing=5, interpSpeedIncreasing=5}
    boneToModify {boneName=dyn_wing_l_d_base}
    rotation {yaw=10}
   }
   n16 = "Transform (Modify) Bone - Bone: dyn_wing_l_c_base"(ComponentPose=@n15, Alpha=@n1, Translation="0,0,0", Rotation="0,7,0", Scale="1,1,1") {
    rotationMode=BMM_Additive
    rotationSpace=BCS_BoneSpace
    scaleSpace=BCS_BoneSpace
    alphaBoolBlend {blendOption=Linear}
    alphaScaleBiasClamp {bInterpResult=true, interpSpeedDecreasing=5, interpSpeedIncreasing=5}
    boneToModify {boneName=dyn_wing_l_c_base}
    rotation {yaw=7}
   }
   n17 = "Transform (Modify) Bone - Bone: dyn_wing_l_b_base"(ComponentPose=@n16, Alpha=@n1, Translation="0,0,0", Rotation="0,3,0", Scale="1,1,1") {
    rotationMode=BMM_Additive
    rotationSpace=BCS_BoneSpace
    scaleSpace=BCS_BoneSpace
    alphaBoolBlend {blendOption=Linear}
    alphaScaleBiasClamp {bInterpResult=true, interpSpeedDecreasing=5, interpSpeedIncreasing=5}
    boneToModify {boneName=dyn_wing_l_b_base}
    rotation {yaw=3}
   }
   n18 = "Transform (Modify) Bone - Bone: dyn_wing_r_01"(ComponentPose=@n17, Alpha=@n1, Translation="0,0,0", Rotation="0.000,20,0.000", Scale="1,1,1") {
    rotationMode=BMM_Additive
    rotationSpace=BCS_BoneSpace
    scaleSpace=BCS_BoneSpace
    alphaBoolBlend {blendOption=Linear}
    alphaScaleBiasClamp {bInterpResult=true, interpSpeedDecreasing=5, interpSpeedIncreasing=5}
    boneToModify {boneName=dyn_wing_r_01}
    rotation {yaw=20}
   }
   n19 = "Transform (Modify) Bone - Bone: dyn_wing_r_e_base"(ComponentPose=@n18, Alpha=@n1, Translation="0,0,0", Rotation="0,-14,0", Scale="1,1,1") {
    rotationMode=BMM_Additive
    rotationSpace=BCS_BoneSpace
    scaleSpace=BCS_BoneSpace
    alphaBoolBlend {blendOption=Linear}
    alphaScaleBiasClamp {bInterpResult=true, interpSpeedDecreasing=5, interpSpeedIncreasing=5}
    boneToModify {boneName=dyn_wing_r_e_base}
    rotation {yaw=-14}
   }
   n20 = "Transform (Modify) Bone - Bone: dyn_wing_r_d_base"(ComponentPose=@n19, Alpha=@n1, Translation="0,0,0", Rotation="0,-10,0", Scale="1,1,1") {
    rotationMode=BMM_Additive
    rotationSpace=BCS_BoneSpace
    scaleSpace=BCS_BoneSpace
    alphaBoolBlend {blendOption=Linear}
    alphaScaleBiasClamp {bInterpResult=true, interpSpeedDecreasing=5, interpSpeedIncreasing=5}
    boneToModify {boneName=dyn_wing_r_d_base}
    rotation {yaw=-10}
   }
   n21 = "Transform (Modify) Bone - Bone: dyn_wing_r_c_base"(ComponentPose=@n20, Alpha=@n1, Translation="0,0,0", Rotation="0,-7,0", Scale="1,1,1") {
    rotationMode=BMM_Additive
    rotationSpace=BCS_BoneSpace
    scaleSpace=BCS_BoneSpace
    alphaBoolBlend {blendOption=Linear}
    alphaScaleBiasClamp {bInterpResult=true, interpSpeedDecreasing=5, interpSpeedIncreasing=5}
    boneToModify {boneName=dyn_wing_r_c_base}
    rotation {yaw=-7}
   }
   n22 = "Transform (Modify) Bone - Bone: dyn_wing_r_b_base"(ComponentPose=@n21, Alpha=@n1, Translation="0,0,0", Rotation="0,-3,0", Scale="1,1,1") {
    rotationMode=BMM_Additive
    rotationSpace=BCS_BoneSpace
    scaleSpace=BCS_BoneSpace
    alphaBoolBlend {blendOption=Linear}
    alphaScaleBiasClamp {bInterpResult=true, interpSpeedDecreasing=5, interpSpeedIncreasing=5}
    boneToModify {boneName=dyn_wing_r_b_base}
    rotation {yaw=-3}
   }
   n23 = RigidBody(ComponentPose=@n22, Alpha=0.4, OverrideWorldGravity=Get GravityOverride()) {
    bOverrideWorldGravity=true
    overridePhysicsAsset=&/M_MED_CoyoteTrail_Pack_feathers_Physics
    alphaBoolBlend {blendOption=Linear}
    baseBoneRef {boneName=root}
    componentAppliedLinearAccClamp=<3000 3000 3000>
    componentLinearAccScale=<0.3 0.3 0.3>
    componentLinearVelScale=<0.3 0.3 0.3>
    overrideWorldGravity {z=-980}
   }
   Output Pose(Result=Component To Local(ComponentPose=@n23)) {layerGroup=None}
  ]}
 EventGraph {nodes [
   Event BlueprintUpdateAnimation()
   n1 = "To Float (Boolean)"(InBool=Get bIsSprinting())
   n2 = OR Boolean!(A=Get bIsTargeting(), B=Get bIsSkydiveDiveMode(), C:bool=Get bIsTacticalSprinting(), D:bool=Get bIsCrouching()) -> (ReturnValue)
   > Set DivePose_alpha(DivePose_alpha:double="Max (Float)"(A="Clamp (Float)"(Value=@n1, Min=0, Max=0.75), B="To Float (Boolean)"(InBool=@n2))) -> (Output_Get:double)
   > Set FlyPose_alpha(FlyPose_alpha:double="To Float (Boolean)"(InBool=Get bIsSkydiveFloating())) -> (Output_Get:double)
   n3 = Get bIsJumping()
   > Set JumpPose_alpha(JumpPose_alpha:double=SelectFloat(A=0, B="To Float (Boolean)"(InBool=@n3), bPickA=Get bIsSkyDiving())) -> (Output_Get:double)
   > n4 = Gravity Override State Function DEPRECATED(CharacterPartAnimInstance=Self-Reference(), GameTimeInSeconds=GetGameTimeInSeconds(), GravityOverrideParams=Get Gravity Override Params())
   > Set GravityOverride(GravityOverride=@n4)
   n5 = Get JumpPose_alpha()
   > "Set FInterp(Alpha, (5: 5))"("FInterp(Alpha, (5: 5))":float="Max (Float)"(A=MapRangeClamped(Value=@n5, InRangeA=0, InRangeB=1, OutRangeB=3), B=Get FlyPose_alpha())) -> (Output_Get:float)
   TryGetPawnOwner()
   TryGetPawnOwner()
  ]}
]
variables [
 DivePose_alpha {type=number}
 "FInterp(Alpha, (5: 5))" {type=number, default=1}
 FlyPose_alpha {type=number}
 Gravity Override Params {type=GravityOverrideParamsStruct, default {onGround_Moving {gravityOverride=<0 0 -650>}, onGround_Standing {gravityOverride=<0 0 -650>}}}
 GravityOverride {type=Vector}
 JumpPose_alpha {type=number}
 Rotation {type=Rotator, default=<rot 0 0 10>}
]
